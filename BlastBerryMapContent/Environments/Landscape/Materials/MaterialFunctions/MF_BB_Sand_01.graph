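# Sand landscape material function for BlastBerry terrain, outputting material attributes plus raw diffuse and roughness.
# Samples Helios diffuse, normal and mask textures with scalable world UVs, feeding normals through a world-aligned normal function.
# A world-Z gradient drives a HeightLerp wetness mask for low-lying sand, apparently darkening color, raising roughness and flattening normals.
# Named reroutes carry height, specular, roughness and wetness; the SandRoughLow/High remap is left unconnected.
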MaterialFunction /BlastBerryMapContent/Environments/Landscape/Materials/MaterialFunctions/MF_BB_Sand_01
graph {nodes [
  n1 = MaterialFunctionCall!(Scale=ScalarParameter() {defaultValue=1024, group=Sand, parameterName=SandScale}) {materialFunction=/BlastBerryMapContent/Game/Athena/Environments/Landscape/MaterialFunctions/Standard/MF_BB_SimpleWorldUV}
  n2 = TextureSample(UVs=@n1.Result) {samplerSource=SSM_Wrap_WorldGroupSettings, texture=/Game/Environments/Helios/Terrain/Textures/T_Helios_Sand_01_D}
  n3 = FunctionOutput(None=@n2.RGB) {outputName=RawDiffuse, sortPriority=100}
  n4 = Constant() {r=0.6}
  n5 = TextureSample(UVs=@n1.Result) {
   samplerSource=SSM_Wrap_WorldGroupSettings
   samplerType=SAMPLERTYPE_Normal
   texture=/Game/Environments/Helios/Terrain/Textures/T_Helios_Sand_01_N
  }
  n6 = StaticBool() {value=true}
  n7 = Add(A=Constant() {r=0.5}, B=NamedRerouteUsage())
  n8 = SetMaterialAttributes!(
   MaterialAttributes
   BaseColor=LinearInterpolate(A=@n3, B=Multiply(A=@n3, B=@n4), Alpha=NamedRerouteUsage())
   Specular=NamedRerouteUsage()
   Roughness=NamedRerouteUsage()
   Normal=LinearInterpolate(
    A=MaterialFunctionCall!(Normals=@n5.RGB, TransformToTangentSpace=@n6, Normal Strength) {materialFunction=/Game/Athena/Apollo/Environments/Landscape/Materials/Biome/CommonMaterials/MF_HQ_WorldAlignedNormal}.Result
    B=Constant3Vector() {constant=<rgba 0 0 1 0>}
    Alpha=Saturate(None=@n7)
   )
   OpacityMask=NamedRerouteUsage()
  ) {attributeSetTypes [
    #1
    #2
    #3
    #4
    #5
   ]}
  FunctionOutput(None=@n8)
  n9 = TextureSample(UVs=@n1.Result) {samplerType=SAMPLERTYPE_Masks, texture=/Game/Environments/Helios/Terrain/Textures/T_Helios_Sand_01_S}
  n10 = NamedRerouteDeclaration(None=FunctionOutput(None=@n9.B) {outputName=RawRough, sortPriority=200}) {name=RawRough, nodeColor=<rgba 1 0 1 1>}
  LinearInterpolate(
   A=ScalarParameter() {group=Sand, parameterName=SandRoughLow}
   B=ScalarParameter() {defaultValue=1, group=Sand, parameterName=SandRoughHigh}
   Alpha=@n10
  )
  n11 = WorldPosition()
  n12 = Constant() {r=-50}
  n13 = Saturate(None=Divide(A=Subtract(A=@n11.Z, B=@n12) {constB=-50}, B=Constant() {r=20}) {constB=50})
  NamedRerouteDeclaration(None=MaterialFunctionCall(Transition Phase=OneMinus(None=@n13), Height Texture=NamedRerouteUsage()) {materialFunction=&HeightLerp}.Alpha) {name=Sand_WetHeight, nodeColor=<rgba 1 0.9609 0 1>}
  NamedRerouteDeclaration(None=@n2.A) {name=Height, nodeColor=<rgba 1 0 1 1>}
  NamedRerouteDeclaration(None=NamedRerouteDeclaration(None=@n9.R) {name=RawSpec, nodeColor=<rgba 1 0 1 1>}) {name=Specular, nodeColor=<rgba 1 0 1 1>}
  n14 = Add(A=@n10, B=ScalarParameter() {defaultValue=0.5, group=Sand, parameterName=SandRoughUnderwaterAdd})
  NamedRerouteDeclaration(None=LinearInterpolate(A=@n10, B=Saturate(None=@n14), Alpha=NamedRerouteUsage())) {name=Rough, nodeColor=<rgba 1 0 1 1>}
 ]}
